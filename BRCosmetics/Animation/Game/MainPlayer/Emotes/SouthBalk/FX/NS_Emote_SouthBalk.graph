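# Niagara system for the SouthBalk emote, rendering engine exhaust as flame meshes with a light, GPU sparks, a glow sprite, smoke, wake meshes and ribbon contrails.
# Flames, sparks, glow and contrails sample a shared skeletal mesh reader to follow the model.
# System acceleration stretches flames and spark travel, while smoke and wake spawning switch on owner-velocity thresholds.
# Time-of-day and front-end modifiers adjust brightness and size.
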NiagaraSystem /BRCosmetics/Animation/Game/MainPlayer/Emotes/SouthBalk/FX/NS_Emote_SouthBalk
dependencies {
 usedDataInterfaces [
  &NiagaraDataInterfaceArrayFloat3
  &NiagaraDataInterfaceCurve
  &NiagaraDataInterfaceColorCurve
  &NiagaraDataInterfaceArrayFloat
  &NiagaraDataInterfaceSkeletalMesh
 ]
 usedDynamicInputs [&FNTimeOfDayColor, &VectorLength]
 usedModules [
  &self:SystemSpawnScript.SystemScriptSource.SystemScriptGraph.NiagaraNodeAssignment_6.SetVariables_#1
  &SystemState
  &self:SystemSpawnScript.SystemScriptSource.SystemScriptGraph.NiagaraNodeAssignment_0.SetVariables_#2
  &self:SystemSpawnScript.SystemScriptSource.SystemScriptGraph.NiagaraNodeAssignment_1.SetVariables_#3
  &self:SystemSpawnScript.SystemScriptSource.SystemScriptGraph.NiagaraNodeAssignment_3.SetVariables_#4
  &InterpolateOverTime
  &self:SystemSpawnScript.SystemScriptSource.SystemScriptGraph.NiagaraNodeAssignment_4.SetVariables_#5
  &self:SystemSpawnScript.SystemScriptSource.SystemScriptGraph.NiagaraNodeAssignment_8.SetVariables_#6
  &self:SystemSpawnScript.SystemScriptSource.SystemScriptGraph.NiagaraNodeAssignment_2.SetVariables_#7
  &self:SystemSpawnScript.SystemScriptSource.SystemScriptGraph.NiagaraNodeAssignment_7.SetVariables_#8
  &self:ScratchModule
 ]
}
emitters {
 ContrailsOpt {
  emitterUpdateScript [
   &EmitterState()
   &SpawnRate(SpawnRate=64, Use Spawn Probability=false) {module=SpawnRate001}
  ]
  particleSpawnScript [
   &FN_SetSpawnGroup(Number of Spawn Groups=2)
   &InitializeRibbon(
    Color=$System.Contrails
    Lifetime=&SelectFloatFromArray(
     Direct Array Index=&SpawnGroup()
     Float Selection Array=DataInterface() {propertyValues {FloatData [0.45,0.45,0.15,0.15]}}
    )
    Ribbon Facing Mode:ENiagara_UnsetDirectSet=::NewEnumerator1
    Ribbon Facing Vector:Vector3f=<1 0 0>
    Ribbon Width=&SelectFloatFromArray(
     Direct Array Index=&SpawnGroup()
     Float Selection Array=DataInterface() {propertyValues {FloatData [35,35,10,10]}}
    )
    UsePositionOffset=false
    Write Material Random=true
   )
   &SkeletalMeshLocation(
    Bone / Socket Index=$Particles.SpawnGroup
    MaskSpawnGroup=false
    Mesh Sampling Type=::NewEnumerator5
    Skeletal Mesh=$System.Skeletal Mesh Reader
    Socket Sampling Mode=::NewEnumerator5
   )
  ]
  particleUpdateScript [
   &ParticleState()
   &ScaleColor(
    Scale Alpha=&Multiply_Float(
     A=&FloatFromCurve(
      FloatCurve=DataInterface() {propertyValues {Curve {keys [
          <at 0 0 RCIM_Cubic>
          {interpMode=RCIM_Cubic, tangentMode=RCTM_User, time=0.2, value=1}
          {interpMode=RCIM_Cubic, tangentMode=RCTM_User, time=1, value=0}
         ]}}}
      Scale Curve=&FNTimeOfDayScalar(Night Scale=0.2, Morning Scale=0.2, Evening Scale=0.2)
     )
     B=1
    )
    Scale RGB=&Multiply_VectorByFloat(
     Vector=&MakeVectorFromLinearColorRGB(Color=&ColorFromCurve(ColorCurve=DataInterface() {propertyValues {
         AlphaCurve {keys []}
         BlueCurve {keys [<at 0 4.8655 RCIM_Cubic>,<at 0.4581 4.8655>,<at 0.8818 6.0417>,<at 0.9992 2.9167 RCIM_Cubic>]}
         GreenCurve {keys [<at 0 10 RCIM_Cubic>,<at 0.4581 10>,<at 0.8818 10>,<at 0.9992 5 RCIM_Cubic>]}
         RedCurve {keys [<at 0 1.1979 RCIM_Cubic>,<at 0.4581 1.1979>,<at 0.8818 0.5>,<at 0.9992 0 RCIM_Cubic>]}
        }}))
     Float=&FloatFromCurve(
      FloatCurve=DataInterface() {propertyValues {Curve {keys [<at 0 1>,<at 0.1 1.2>,<at 0.3 1>,<at 0.5 0.5>,<at 0.7 1.5>,<at 1 1>]}}}
      Scale Curve=0.8
     )
    )
    ScaleA=true
    ScaleRGB=false
   )
   &SolveForcesAndVelocity(Clamp Velocity=false, Limit Acceleration=false)
   &ScaleColor(Scale Alpha=0.5, ScaleA=true, ScaleRGB=true) {module=ScaleColor001}
  ]
  rendererClasses [&NiagaraRibbonRendererProperties]
  renderers [{rendererClass=&NiagaraRibbonRendererProperties}]
 }
 EngineFlames {
  emitterUpdateScript [
   &EmitterState(
    Inactive Response=::NewEnumerator1
    Life Cycle Mode=::NewEnumerator1
    Loop Behavior=::NewEnumerator1
    Loop Duration Mode=::NewEnumerator1
    UseLoopDelay=false
   )
   &SpawnBurst_Instantaneous(Spawn Group=1, Spawn Time=0.3, Use Spawn Probability=false, UseLoopCountLimit=false) {module=SpawnBurst_Instantaneous001}
   &SpawnBurst_Instantaneous(Spawn Count=2, Spawn Time=0.3, Use Spawn Probability=true)
  ]
  particleSpawnScript [
   &FN_SetSpawnGroup(Number of Spawn Groups=2)
   &InitializeParticle(
    Mesh Scale=&Multiply_Vector(
     A=&SelectVectorFromArray(
      Direct Array Index=$Particles.SpawnGroup
      Vector Selection Array=DataInterface() {propertyValues {FloatData [<0.35 0.35 0.25>, <0.35 0.35 0.25>, <0.65 0.65 0.65>]}}
     )
     B=&VectorFromFloat(Value=$User.GlobalFlameScale)
    )
    Mesh Scale Mode=::NewEnumerator4
    Sprite Size Mode=::NewEnumerator3
    Uniform Sprite Size=220
    UsePositionOffset=false
    Write Material Random=true
   )
   &InitialMeshOrientation(Orientation Axis=<0 0 1>, Orientation Vector=<0 -1 0>, Use Orientation Vector=true, Use Rotation Vector=true) {module=InitialMeshOrientation001}
   &InitialMeshOrientation(Mesh Orientation Mode=::NewEnumerator1, Rotation=<0.02 0 0>, Use Rotation Vector=true) {enabled=false}
   /SetVariables(
    Particles.Flame2Scale:Vector3f=&Multiply_Vector(A=$Particles.Scale, B=<1 1 0.9>)
    Particles.FlameCoreScale:Vector3f=&Multiply_Vector(A=$Particles.Scale, B=<0.65 0.6 0.6>)
   ) {isSetParameters=true}
   /SetVariables(Particles.VisibilityTag=$Emitter.SpawnGroup) {isSetParameters=true}
  ]
  particleUpdateScript [
   &ParticleState(Kill Particles When Lifetime Has Elapsed=false)
   &SkeletalMeshLocation(
    Bone Sampling Mode=::NewEnumerator4
    MaskSpawnGroup=false
    Sampled Position Offset=&MaskVector(Mask Source=1, Mask Parameter=$Particles.VisibilityTag, Mask=0, Masked Vector=<0 -50 -9>, Passthrough Vector=<0 -60 -9>)
    Skeletal Mesh=$System.Skeletal Mesh Reader
   )
   &SolveForcesAndVelocity(Clamp Velocity=false, Limit Acceleration=false)
   &ScaleColor(
    Scale Alpha=&FloatFromCurve(FloatCurve=DataInterface() {propertyValues {Curve {keys [<at 0 0 RCIM_Cubic>,<at 0.5 1 RCIM_Cubic>]}}})
    ScaleA=true
    ScaleRGB=true
   )
   &ScaleMeshSize(Scale Factor=&MakeVector(
     X=&RandomRangeFloat(Minimum=0.11, Maximum=0.14, Evaluation Type=::NewEnumerator1)
     Y=&RandomRangeFloat(Minimum=0.11, Maximum=0.14, Evaluation Type=::NewEnumerator1)
     Z=&FloatFromCurve(
      FloatCurve=DataInterface() {propertyValues {Curve {keys [<at 0.1 2 RCIM_Cubic>,<at 0.7 1 RCIM_Cubic>]}}}
      Scale Curve=&Multiply_Float(A=0.09, B=$System.Acceleration)
     )
    ))
   &ScaleMeshSize(Scale Factor=&Multiply_Vector(A=<1 1 2.4>, B=$Engine.Owner.Scale)) {module=ScaleMeshSize001}
   /SetVariables(Particles.FlameCoreScale:Vector3f=&Multiply_Vector(
     A=&RandomRangeVector(Minimum=<0.01 0.01 0.08>, Maximum=<0.04 0.04 0.08>, Evaluation Type=::NewEnumerator1)
     B=&Multiply_Vector(
      A=&VectorFromFloat(Value=$User.GlobalFlameScale)
      B=&MakeVector(
       X=1
       Y=1
       Z=&FloatFromCurve(FloatCurve=DataInterface() {propertyValues {Curve {keys [<at 0.1 2 RCIM_Cubic>,<at 0.5 1 RCIM_Cubic>]}}}, Scale Curve=$System.Acceleration)
      )
     )
    )) {isSetParameters=true}
   /SetVariables(Particles.Color02:LinearColor=&MultiplyLinearColor(Linear Color=<rgba 1 1 1 1>, Scale Factor=&MakeVector4_FromFloat(Float=1))) {isSetParameters=true}
   &DynamicMaterialParameters(Module.Param0WriteEnabled=true, Module.Param1WriteEnabled=true, Module.Param2WriteEnabled=true, Module.Param3WriteEnabled=true)
   &FN_Mesh_Attributes(
    Color=&MultiplyLinearColor(
     Linear Color=&MakeCustomLinearColorFromBool(Bool=$User.bIsFrontEndPreview, True Color=<rgba 4 2.129 0.5147 1>, False Color=<rgba 4 2.129 0.5147 1>)
     Scale Factor=&MakeVector4_FromFloat(Float=$System.TOD)
    )
    Create Color Attribute=true
    Create Scale Attribute=true
    Scale=&Multiply_Vector(
     A=&RandomRangeVector(Minimum=<0.04 0.04 0.1>, Maximum=<0.05 0.05 0.1>, Evaluation Type=::NewEnumerator1)
     B=&Multiply_Vector(
      A=&VectorFromFloat(Value=$User.GlobalFlameScale)
      B=&MakeVector(
       X=1
       Y=1
       Z=&FloatFromCurve(FloatCurve=DataInterface() {propertyValues {Curve {keys [<at 0.1 2.5 RCIM_Cubic>,<at 0.5 1 RCIM_Cubic>]}}}, Scale Curve=$System.Acceleration)
      )
     )
    )
    Write to Module Namespace=true
   )
   &FN_Light_Attributes(
    Color=<rgba 20 3.8389 0.9621 1>
    Create Color Attribute=true
    Create Exponent Attribute=true
    Create Radius Attribute=true
    Exponent=1
    Radius=&FloatFromCurve(
     FloatCurve=DataInterface() {propertyValues {Curve {keys [
         <at 0 0 RCIM_Cubic>
         {arriveTangent=2.3203, interpMode=RCIM_Cubic, leaveTangent=2.3203, tangentMode=RCTM_User, time=0.0148, value=2}
         {arriveTangent=-0.577, interpMode=RCIM_Cubic, leaveTangent=-0.577, tangentMode=RCTM_User, time=0.8, value=1}
        ]}}}
     Scale Curve=30
    )
    Scale Color=&FloatFromCurve(
     FloatCurve=DataInterface() {propertyValues {Curve {keys [
         <at 0 0 RCIM_Cubic>
         {arriveTangent=0.4006, interpMode=RCIM_Cubic, leaveTangent=0.4006, tangentMode=RCTM_User, time=0.1, value=3}
         {arriveTangent=-2.839, interpMode=RCIM_Cubic, leaveTangent=-2.839, tangentMode=RCTM_User, time=0.8, value=1}
        ]}}}
     Scale Curve=&Waveform("Waveform [1]"=::NewEnumerator10, "[1] Frequency"=50, "[1] Amplitude Min/Max"=<0.5 1>)
    )
   )
  ]
  rendererClasses [&NiagaraMeshRendererProperties, &NiagaraLightRendererProperties]
  renderers [
   {rendererClass=&NiagaraMeshRendererProperties, rendererIndex=0}
   {rendererClass=&NiagaraMeshRendererProperties, rendererIndex=1}
   {rendererClass=&NiagaraMeshRendererProperties, rendererIndex=2}
   {rendererClass=&NiagaraLightRendererProperties, rendererIndex=3}
  ]
 }
 LargeWake {
  emitterUpdateScript [
   &EmitterState(
    Delay First Loop Only=true
    Inactive Response=::NewEnumerator1
    Life Cycle Mode=::NewEnumerator1
    Loop Delay=0.4
    Loop Duration=1
    UseLoopDelay=true
   )
   &SpawnRate(
    SpawnRate=&Lerp_Float(
     A=4.5
     B=0
     Alpha=&MakeCustomFloatFromBool(
      Bool=&SetBoolByFloatComparison(
       A=&Add_Float(
        A=&Add_Float(
         A=&Abs_Float(Float:NiagaraFloat=&MakeFloatFromVector(Vector=$Engine.Owner.Velocity, Channel=::NewEnumerator0))
         B=&Abs_Float(Float:NiagaraFloat=&MakeFloatFromVector(Vector=$Engine.Owner.Velocity, Channel=::NewEnumerator2))
        )
        B=&Abs_Float(Float:NiagaraFloat=&MakeFloatFromVector(Vector=$Engine.Owner.Velocity, Channel=::NewEnumerator1))
       )
       B=150
      )
      True Float=&ClampFloat(Float=1)
      False Float=&ClampFloat(Float=0)
     )
    )
    Use Spawn Probability=false
   ) {module=SpawnRate001}
  ]
  particleSpawnScript [
   &InitializeParticle(
    Color=$System.Linear Color
    Lifetime Max=1.8
    Lifetime Min=1.4
    Lifetime Mode=::NewEnumerator1
    Mesh Scale=&Multiply_VectorByFloat(
     Vector=&RandomRangeVector(Minimum=<1 1 0.5>, Maximum=<1 1 1>)
     Float=&RandomRangeFloat(Minimum=0.65, Maximum=1.1)
    )
    Mesh Scale Mode=::NewEnumerator4
    Position Offset=&MakeVector(X=0, Y=0, Z=0)
    UsePositionOffset=true
    Write Material Random=true
   )
   &InitialMeshOrientation(
    Mesh Orientation Mode=::NewEnumerator2
    Orientation Coordinate Space=::World
    Rotation=&MakeVector(
     X=&MakeCustomFloatFromBool(Bool=&RandomBool(), True Float=0, False Float=0)
     Y=&MakeCustomFloatFromBool(Bool=&RandomBool(), True Float=0, False Float=0)
     Z=&RandomRangeFloat(Minimum=-1, Maximum=1)
    )
    Use Rotation Vector=true
   )
  ]
  particleUpdateScript [
   &ParticleState()
   &DynamicMaterialParameters(
    Index 0 Param 1=&FloatFromCurve(FloatCurve=DataInterface() {propertyValues {Curve {keys [
         <at 0 0.2 RCIM_Cubic>
         {interpMode=RCIM_Cubic, tangentMode=RCTM_User, time=0.3, value=0.45}
         {interpMode=RCIM_Cubic, tangentMode=RCTM_User, time=1, value=0}
        ]}}})
    Module.Param0WriteEnabled=true
    Module.Param1WriteEnabled=true
    Module.Param2WriteEnabled=true
    Module.Param3WriteEnabled=true
   )
   &ScaleMeshSize(Scale Factor=&Multiply_VectorByFloat(
     Vector=&MakeVector(
      X=1
      Y=1
      Z=&FloatFromCurve(FloatCurve=DataInterface() {propertyValues {Curve {keys [<at 0.15 1>,<at 0.6 0>]}}})
     )
     Float=&FloatFromCurve(
      FloatCurve=DataInterface() {propertyValues {Curve {keys [
          {arriveTangent=2.3149, interpMode=RCIM_Cubic, leaveTangent=2.3149, tangentMode=RCTM_User, time=0, value=0}
          {arriveTangent=0.9995, interpMode=RCIM_Cubic, leaveTangent=0.9995, tangentMode=RCTM_User, time=0.3, value=0.45}
          {arriveTangent=-0.0405, interpMode=RCIM_Cubic, leaveTangent=-0.0405, tangentMode=RCTM_User, time=1, value=1}
         ]}}}
      Scale Curve=2.5
     )
    ))
   &UpdateMeshOrientation(
    Rotation Rate=&MakeCustomFloatFromBool(Bool=&RandomBool(), True Float=0.5, False Float=-0.5)
    Rotation Vector=<0 0 1>
   ) {enabled=false}
   &ScaleColor(
    Scale Alpha=&FloatFromCurve(FloatCurve=DataInterface() {propertyValues {Curve {keys [
         {interpMode=RCIM_Cubic, tangentMode=RCTM_User, time=0, value=0.2}
         {interpMode=RCIM_Cubic, leaveTangent=-2.4875, tangentMode=RCTM_Break, time=0.17, value=1}
         {tangentMode=RCTM_User, time=1, value=0}
        ]}}})
    Scale RGB=&Multiply_VectorByFloat(
     Vector=&VectorFromFloat(Value=&FloatFromCurve(FloatCurve=DataInterface() {propertyValues {Curve {keys [<at 0 1>,<at 1 1>]}}}))
     Float=&FN_FrontEndFloatModifier(Gameplay=1, Front End=2)
    )
    ScaleA=true
    ScaleRGB=true
   )
   /SetVariables(Particles.Mesh soft color:LinearColor=&MakeLinearColorFromVectorAndFloat(
     "Vector (RGB)"=&Multiply_VectorByFloat(Vector=<1 1 1>, Float=&FN_FrontEndFloatModifier(Gameplay=1, Front End=2))
     "Float (Alpha)"=&Multiply_Float(
      A=&FloatFromCurve(FloatCurve=DataInterface() {propertyValues {Curve {keys [
           {interpMode=RCIM_Cubic, tangentMode=RCTM_User, time=0, value=0}
           {interpMode=RCIM_Cubic, tangentMode=RCTM_User, time=0.5, value=0.2}
           {tangentMode=RCTM_User, time=1, value=0}
          ]}}})
      B=1
     )
    )) {enabled=false, isSetParameters=true}
   &Color(Color=<rgba 0.1875 0.1875 0.1875 1>) {enabled=false}
  ]
  rendererClasses [&NiagaraMeshRendererProperties]
  renderers [
   {rendererClass=&NiagaraMeshRendererProperties, rendererIndex=0}
   {rendererClass=&NiagaraMeshRendererProperties, rendererIndex=1}
  ]
 }
 Minimal {
  emitterUpdateScript [
   &EmitterState()
   &SpawnBurst_Instantaneous(Use Spawn Probability=false, UseLoopCountLimit=false)
  ]
  particleSpawnScript [
   &InitializeParticle(
    Color=<rgba 50 1.6597 0 50000>
    Lifetime=5
    Sprite Size Mode=::NewEnumerator3
    Uniform Sprite Size=20
    UsePositionOffset=false
    Write Material Random=true
   )
   &SpriteFacingAndAlignment(Sprite Facing=<0.001 6 0>, WriteAlignment=true, WriteFacing=true)
  ]
  particleUpdateScript [
   &ParticleState()
   &SkeletalMeshLocation(Bone Sampling Mode=::NewEnumerator4, MaskSpawnGroup=false, Sampled Position Offset=<0 -55 -9>, Skeletal Mesh=$System.Skeletal Mesh Reader)
   &ScaleColor(
    Scale Alpha=&FloatFromCurve(FloatCurve=DataInterface() {propertyValues {Curve {keys [<at 0.35 0>,<at 0.4 1>]}}}, CurveIndex=$Emitter.Age)
    ScaleA=true
    ScaleRGB=true
   )
  ]
  rendererClasses [&NiagaraSpriteRendererProperties]
  renderers [{rendererClass=&NiagaraSpriteRendererProperties}]
 }
 Smoke {
  emitterUpdateScript [
   &EmitterState()
   &SpawnRate(
    SpawnRate=&Lerp_Float(
     A=&RandomRangeFloat(Minimum=0, Maximum=10, Evaluation Type=::NewEnumerator1)
     B=0
     Alpha=&MakeCustomFloatFromBool(
      Bool=&SetBoolByFloatComparison(
       A=&Add_Float(
        A=&Add_Float(
         A=&Abs_Float(Float:NiagaraFloat=&MakeFloatFromVector(Vector=$Engine.Owner.Velocity, Channel=::NewEnumerator0))
         B=&Abs_Float(Float:NiagaraFloat=&MakeFloatFromVector(Vector=$Engine.Owner.Velocity, Channel=::NewEnumerator2))
        )
        B=&Abs_Float(Float:NiagaraFloat=&MakeFloatFromVector(Vector=$Engine.Owner.Velocity, Channel=::NewEnumerator1))
       )
       B=170
      )
      True Float=&ClampFloat(Float=1)
      False Float=&ClampFloat(Float=0)
     )
    )
    Use Spawn Probability=false
   ) {module=SpawnRate001}
  ]
  particleSpawnScript [
   /SetVariables(Particles.Random=&RandomRangeFloat(Minimum=0, Maximum=1)) {isSetParameters=true}
   &CameraOffset(Camera Offset Amount=30)
   &DynamicMaterialParameters(
    Index 0 Param 2=&UniformAOrBFloat(A=0)
    Module.Param0WriteEnabled=true
    Module.Param1WriteEnabled=true
    Module.Param2WriteEnabled=true
    Module.Param3WriteEnabled=true
   ) {module=DynamicMaterialParameters001}
   &InitializeParticle(
    Color=$System.SmokeCol
    Lifetime=1.1
    Sprite Rotation Mode=::NewEnumerator2
    Sprite Size Mode=::NewEnumerator3
    Uniform Sprite Size=&Lerp_Float(A=120, B=170, Alpha=$Particles.Random)
    UsePositionOffset=true
    Write Material Random=true
   ) {module=InitializeParticle001}
   &ApplyInitialForces()
   &InitialMeshOrientation(Mesh Orientation Mode=::NewEnumerator2, Orientation Coordinate Space=::World, Rotation=<0 -0.25 0>, Use Rotation Vector=true)
  ]
  particleUpdateScript [
   &ParticleState()
   &ScaleSpriteSize(Uniform Curve Sprite Scale=DataInterface() {propertyValues {Curve {keys [
        {arriveTangent=1.6815, interpMode=RCIM_Cubic, leaveTangent=1.6815, tangentMode=RCTM_User, time=0.2, value=0.5}
        {arriveTangent=0.1063, interpMode=RCIM_Cubic, leaveTangent=0.1063, tangentMode=RCTM_User, time=0.55, value=1}
       ]}}})
   &SolveForcesAndVelocity(Clamp Velocity=false, Limit Acceleration=false)
   &AlignSpriteToMeshOrientation(SetAlignment=true, SetFacing=true)
   &DynamicMaterialParameters(
    Index 0 Param 1=&FloatFromCurve(FloatCurve=DataInterface() {propertyValues {Curve {keys [
         {arriveTangent=1.7661, interpMode=RCIM_Cubic, leaveTangent=1.7661, tangentMode=RCTM_User, time=0, value=0}
         {arriveTangent=0.0224, interpMode=RCIM_Cubic, leaveTangent=0.0224, tangentMode=RCTM_User, time=1, value=1}
        ]}}})
    Index 0 Param 2=1
    Module.Param0WriteEnabled=true
    Module.Param1WriteEnabled=true
    Module.Param2WriteEnabled=true
    Module.Param3WriteEnabled=true
   )
   &ScaleColor(
    Scale Alpha=&FloatFromCurve(FloatCurve=DataInterface() {propertyValues {Curve {keys [<at 0 0>, {tangentMode=RCTM_User, time=0.75, value=1}, {tangentMode=RCTM_User, time=1, value=0}]}}})
    Scale RGB=&MakeVectorFromLinearColorRGB(Color=&ColorFromCurve(ColorCurve=DataInterface() {propertyValues {
        AlphaCurve {keys [<at 0 0>,<at 0.4487 1 RCIM_Cubic>]}
        BlueCurve {keys [
          <at 0.1811 0.6425 RCIM_Cubic>
          {arriveTangent=-0.3057, interpMode=RCIM_Cubic, leaveTangent=-0.3057, time=0.4806, value=1}
          <at 0.7837 0.4583>
         ]}
        GreenCurve {keys [
          <at 0.1811 0.6562 RCIM_Cubic>
          {arriveTangent=-0.3285, interpMode=RCIM_Cubic, leaveTangent=-0.3285, time=0.4806, value=1}
          <at 0.7837 0.4583>
         ]}
        RedCurve {keys [
          <at 0.1811 0.6388 RCIM_Cubic>
          {arriveTangent=-0.2996, interpMode=RCIM_Cubic, leaveTangent=-0.2996, time=0.4806, value=1}
          <at 0.7837 0.4583>
         ]}
       }}))
    ScaleA=true
    ScaleRGB=true
   )
  ]
  rendererClasses [&NiagaraSpriteRendererProperties]
  renderers [{rendererClass=&NiagaraSpriteRendererProperties}]
 }
 Sparks_Forward {
  simTarget=GPUComputeSim
  emitterUpdateScript [
   &EmitterState(
    Delay First Loop Only=true
    Inactive Response=::NewEnumerator1
    Life Cycle Mode=::NewEnumerator1
    Loop Delay=0.3
    Loop Duration=1
    UseLoopDelay=true
   )
   &SpawnRate(SpawnRate=500, Use Spawn Probability=true)
  ]
  particleSpawnScript [
   &PartitionParticles(Number Of Partitions=1)
   &InitializeParticle(
    Color=<rgba 1.5 1.2105 0.384 1>
    Lifetime Max=0.6
    Lifetime Min=0.4
    Lifetime Mode=::NewEnumerator1
    Sprite Size Max=<0.5 3>
    Sprite Size Min=<0.5 3>
    Sprite Size Mode=::NewEnumerator2
    UsePositionOffset=false
    Write Material Random=true
   )
   &ShapeLocation(
    MaskSpawnGroup=false
    Non Uniform Scale=<5 1 5>
    Offset=<0 -50 -8>
    Offset Mode=::NewEnumerator0
    Override Seed=false
    Sphere Radius=2
    UseOwnerScale=true
   )
   &AddVelocity(Velocity=<0 -1 0>, Velocity Speed Scale=1.3) {module=AddVelocity001}
   /SetVariables(Particles.AccelerationAtSpawn:NiagaraFloat=&RemapRange(Input Value To Remap=$System.Acceleration, Input Min=0.7, Input Max=1, Output Min=0, Output Max=1)) {isSetParameters=true}
  ]
  particleUpdateScript [
   &ParticleState()
   &SkeletalMeshLocation(
    Bone Sampling Mode=::NewEnumerator4
    MaskSpawnGroup=false
    Sampled Position Offset=&AddVectorToPosition(
     Position=$Particles.ShapeLocation.ShapePosition
     Vector=&MakeVector(
      X=&FloatFromCurve(
       FloatCurve=DataInterface() {propertyValues {Curve {defaultValue=1, keys [{tangentMode=RCTM_User, time=0, value=0}, {tangentMode=RCTM_User, time=1, value=1}]}}}
       Scale Curve=&RandomRangeFloat(Minimum=10, Maximum=-10)
      )
      Y=&FloatFromCurve(
       FloatCurve=DataInterface() {propertyValues {Curve {keys [<at 0 0>,<at 1 1>]}}}
       Scale Curve=&FloatFromCurve(
        FloatCurve=DataInterface() {propertyValues {Curve {keys [<at 0 2>,<at 1 1>]}}}
        CurveIndex=$Emitter.Age
        Scale Curve=&Lerp_Float(A=-35, B=-100, Alpha=$Particles.AccelerationAtSpawn)
       )
      )
      Z=&FloatFromCurve(
       FloatCurve=DataInterface() {propertyValues {Curve {defaultValue=1, keys [{tangentMode=RCTM_User, time=0, value=0}, {tangentMode=RCTM_User, time=1, value=1}]}}}
       Scale Curve=&RandomRangeFloat(Minimum=10, Maximum=-10)
      )
     )
    )
    Skeletal Mesh=$System.Skeletal Mesh Reader
   )
   /SetVariables(Particles.Glow03_Color=&MakeLinearColorFromVectorAndFloat(
     "Vector (RGB)"=&MakeVectorFromLinearColorRGB(Color=&MultiplyLinearColor(
       Linear Color=&ColorFromCurve(ColorCurve=DataInterface() {propertyValues {
          AlphaCurve {keys []}
          BlueCurve {keys [<at 0 0.3733 RCIM_Cubic>,<at 0.5482 0.2623 RCIM_Cubic>]}
          GreenCurve {keys [<at 0 1.8979 RCIM_Cubic>,<at 0.5482 0 RCIM_Cubic>]}
          RedCurve {keys [<at 0 3 RCIM_Cubic>,<at 0.5482 1 RCIM_Cubic>]}
         }})
       Scale Factor=&MakeVector4_FromFloat(Float=$System.TOD)
      ))
     "Float (Alpha)"=&Multiply_Float(A=0.2, B=&FN_FrontEndFloatModifier(Gameplay=1, Front End=0.1))
    )) {isSetParameters=true}
   /SetVariables(Particles.Glow03_Size=&Multiply_Vector2D_ByFloat(
     Float=&FloatFromCurve(
      FloatCurve=DataInterface() {propertyValues {Curve {keys [
          <at 0 0 RCIM_Cubic>
          <at 0.2 1.3 RCIM_Cubic>
          {arriveTangent=-3.0745, interpMode=RCIM_Cubic, leaveTangent=-3.0745, tangentMode=RCTM_User, time=1, value=0}
         ]}}}
      Scale Curve=&FN_FrontEndFloatModifier(Gameplay=5, Front End=15)
     )
     Vector2D=<0.5 1.5>
    )) {isSetParameters=true}
   &ScaleColor(
    Scale Alpha=&FloatFromCurve(FloatCurve=DataInterface() {propertyValues {Curve {keys [<at 0.2 0>,<at 0.3 0.8>,<at 1 0>]}}})
    ScaleA=true
    ScaleRGB=true
   )
   &ScaleSpriteSize(Scale Factor=&Vector2DFromFloat(Value=&FloatFromCurve(
      FloatCurve=DataInterface() {propertyValues {Curve {keys [
          <at 0 0 RCIM_Cubic>
          {interpMode=RCIM_Cubic, leaveTangent=-3.0296, tangentMode=RCTM_Break, time=0.2, value=1.2}
          {arriveTangent=-1.8107, interpMode=RCIM_Cubic, leaveTangent=-1.8107, tangentMode=RCTM_User, time=1, value=0}
         ]}}}
      Scale Curve=&FN_FrontEndFloatModifier(Gameplay=0.5, Front End=0.3)
     ))) {module=ScaleSpriteSize001}
   &Drag(Use Linear Drag=true, Use Rotational Drag=false) {enabled=false}
   &SolveForcesAndVelocity(Clamp Velocity=false, Limit Acceleration=false)
  ]
  rendererClasses [&NiagaraSpriteRendererProperties]
  renderers [
   {rendererClass=&NiagaraSpriteRendererProperties}
   {rendererClass=&NiagaraSpriteRendererProperties, rendererIndex=1}
  ]
 }
}
summary {
 systemName=NS_Emote_SouthBalk
 emitters [
  {emitterName=EngineFlames, rendererClasses [&NiagaraMeshRendererProperties, &NiagaraLightRendererProperties]}
  {emitterName=Sparks_Forward, simTarget=GPUComputeSim, rendererClasses [&NiagaraSpriteRendererProperties]}
  {emitterName=Minimal, rendererClasses [&NiagaraSpriteRendererProperties]}
  {emitterName=Smoke, rendererClasses [&NiagaraSpriteRendererProperties]}
  {emitterName=LargeWake, rendererClasses [&NiagaraMeshRendererProperties]}
  {emitterName=ContrailsOpt, rendererClasses [&NiagaraRibbonRendererProperties]}
 ]
 userVariables [
  User.bIsFrontEnd {type=NiagaraBool}
  User.bIsFrontEndPreview {type=NiagaraBool}
  User.Flame_Size {type=NiagaraFloat, defaultValue=1, description=""}
  User.GlobalFlameScale {type=NiagaraFloat, defaultValue=9, description=""}
  User.JumpCounter {type=NiagaraInt32}
 ]
}
