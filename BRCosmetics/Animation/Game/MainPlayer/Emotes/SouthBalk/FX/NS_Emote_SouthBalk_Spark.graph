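# Niagara system for a small one-shot spark burst in the SouthBalk emote.
# A single sprite emitter, gated by a system-level Activate flag, spawns a few short-lived particles after the user-set ExplosionDelay.
# Sparks launch from a small sphere with randomized directional velocity, then slow under drag and fall with gravity.
# Sprites stretch with speed and thin over life while an HDR color curve cools from bright red-orange to dim red.
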NiagaraSystem /BRCosmetics/Animation/Game/MainPlayer/Emotes/SouthBalk/FX/NS_Emote_SouthBalk_Spark
dependencies {
 usedDataInterfaces [&NiagaraDataInterfaceCurve, &NiagaraDataInterfaceVector2DCurve, &NiagaraDataInterfaceColorCurve]
 usedDynamicInputs [&MaskLinearColorBySpawnGroup, &SetBoolByFloatComparison]
 usedModules [
  &self:SystemSpawnScript.SystemScriptSource.SystemScriptGraph.NiagaraNodeAssignment_0.SetVariables_#1
  &self:SystemSpawnScript.SystemScriptSource.SystemScriptGraph.NiagaraNodeAssignment_1.SetVariables_#2
  &SystemState
  &self:SystemSpawnScript.SystemScriptSource.SystemScriptGraph.NiagaraNodeAssignment_2.SetVariables_#3
 ]
}
emitters {Directional_sparks001 {
  emitterUpdateScript [
   &EmitterState(Inactive Response=::NewEnumerator2, Life Cycle Mode=::NewEnumerator1, Loop Behavior=::NewEnumerator1, UseLoopDelay=false)
   &self:Activate_Emitter_08(Active:NiagaraBool=$System.Activate, Lifetime=1.3) {module=Activate_Emitter_08}
   &SpawnBurst_Instantaneous(
    Spawn Count=&RandomRangeInt(Minimum=4, Maximum=8)
    Spawn Time=$User.ExplosionDelay
    Use Spawn Probability=false
   )
  ]
  particleSpawnScript [
   &InitializeParticle(
    Color Mode=::NewEnumerator0
    Lifetime Max=0.4
    Lifetime Min=0.2
    Lifetime Mode=::NewEnumerator1
    Sprite Size Max=<1.3333 3.3333>
    Sprite Size Min=<0.6667 1.3333>
    Sprite Size Mode=::NewEnumerator2
    UsePositionOffset=false
    Write Material Random=true
   ) {module=InitializeParticle001}
   &ShapeLocation(MaskSpawnGroup=false, Override Seed=false, Sphere Radius=20, UseOwnerScale=true) {module=ShapeLocation001}
   &CameraOffset(Camera Offset Amount=15)
   &AddVelocity(
    Velocity=&RandomRangeVector(Minimum=<1 -1.5 2>, Maximum=<-1 -2.3 0.5>)
    Velocity Speed Scale=&RandomRangeFloat(Minimum=100, Maximum=50)
   )
   &AddVelocity(
    Cone Angle=200
    Distribution Along Cone Axis=90
    Module.UseConeFalloff=true
    Module.UseVelDistribution=true
    Override Seed=false
    Speed Falloff From Cone Axis=0.5
    Velocity Mode=::NewEnumerator2
    Velocity Speed=&RandomRangeFloat(Minimum=50, Maximum=150)
   ) {enabled=false, module=AddVelocity001}
  ]
  particleUpdateScript [
   &ParticleState()
   &Drag(Drag=&RandomRangeFloat(Minimum=1, Maximum=2), Use Linear Drag=true, Use Rotational Drag=false)
   &GravityForce(Gravity=&RandomRangeVector(Minimum=<0 0 -200>, Maximum=<0 0 -500>))
   &SolveForcesAndVelocity(Clamp Velocity=false, Limit Acceleration=false)
   &ScaleSpriteSizeBySpeed(Max Scale Factor=<1 10>, Min Scale Factor=<0.65 0.65>)
   &ScaleSpriteSize(Scale Factor=&Vector2DFromCurve(Vector2Curve=DataInterface() {propertyValues {XCurve {keys [<at 0 1>,<at 1 1>]}, YCurve {keys [<at 0 1>,<at 1 0>]}}}, ScaleCurve=<1 1>))
   &ScaleColor(
    Scale RGB=&MakeVectorFromLinearColorRGB(Color=&ColorFromCurve(ColorCurve=DataInterface() {propertyValues {
        AlphaCurve {keys [<at 0 1 RCIM_Cubic>]}
        BlueCurve {keys [<at 0 1.1031 RCIM_Cubic>,<at 1 0 RCIM_Cubic>]}
        GreenCurve {keys [<at 0 5.688 RCIM_Cubic>,<at 1 0 RCIM_Cubic>]}
        RedCurve {keys [<at 0 120 RCIM_Cubic>,<at 1 1 RCIM_Cubic>]}
       }}))
    ScaleA=true
    ScaleRGB=true
   )
  ]
  rendererClasses [&NiagaraSpriteRendererProperties]
  renderers [{rendererClass=&NiagaraSpriteRendererProperties}]
 }}
summary {
 systemName=NS_Emote_SouthBalk_Spark
 emitters [{emitterName=Directional_sparks001, rendererClasses [&NiagaraSpriteRendererProperties]}]
 userVariables [
  User.bIsFrontEnd {type=NiagaraBool}
  User.bIsFrontEndPreview {type=NiagaraBool}
  User.ExplosionDelay {type=NiagaraFloat, defaultValue=0, description=""}
  User.HitGround {type=NiagaraBool, defaultValue=true}
 ]
}
